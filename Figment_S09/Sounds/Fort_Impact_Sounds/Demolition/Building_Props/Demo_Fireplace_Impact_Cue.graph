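# Sound cue for impacts on a fireplace building prop in the demolition impact sound set.
# A mixer layers a fireplace impact wave with one of three equally weighted fire pit impact waves picked at random.
# Each layer has its own pitch and volume modulation, with the fire pit layer pitched higher.
# Playback uses close demolition attenuation, an impact sound class and an object-impact concurrency set.
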SoundCue /Figment_S09/Sounds/Fort_Impact_Sounds/Demolition/Building_Props/Demo_Fireplace_Impact_Cue
root=Mixer_0
nodes [
 {id=Mixer_0, children [Modulator_2,Modulator_1], properties {inputVolume [1,1]}}
 {id=Modulator_1, children [WavePlayer_32], properties {pitchMax=1.2, pitchMin=0.85, volumeMin=0.6383}}
 {id=Modulator_2, children [Random_5], properties {pitchMax=1.5, pitchMin=1.1, volumeMin=0.85}}
 {id=Random_5, children [WavePlayer_28, WavePlayer_29, WavePlayer_30], properties {weights [1,1,1]}}
 {id=WavePlayer_28, properties {soundWaveAssetPtr=/Figment_S09/Sounds/Fort_Impact_Sounds/Demolition/FirePit_Impact03}}
 {id=WavePlayer_29, properties {soundWaveAssetPtr=/Figment_S09/Sounds/Fort_Impact_Sounds/Demolition/FirePit_Impact04}}
 {id=WavePlayer_30, properties {soundWaveAssetPtr=/Figment_S09/Sounds/Fort_Impact_Sounds/Demolition/FirePit_Impact05}}
 {id=WavePlayer_32, properties {soundWaveAssetPtr=/Figment_S09/Sounds/Fort_Impact_Sounds/Demolition/Fireplace_Impact01}}
]
properties {
 attenuationSettings=/Figment_S09/Sounds/Attenuation/Fort_Demolition_Close_Att
 maxDistance=2000
 soundClassObject=/Game/Sounds/FortSoundClassesAndModes/Fort_ImpactSounds_DuckDefault_PSM
 volumeMultiplier=1
 concurrencyOverrides {resolutionRule=PreventNew}
 concurrencySet [&CONC_Object_Impacts_FourMax]
}
